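# Niagara burst effect for the heart spawn moment of the Disconnect player emote.
# Short-lived emitters layer a light, mesh shockwave, flare with flash, multi-sprite glow, SubUV shine and drag-slowed dust, plus an "Empty" sprite-and-mesh element that scales in quickly.
# User Color Array tints shine and dust.
# Time-of-day scalars adjust glow, dust and shine, while a front-end modifier resizes the flare separately for menus and gameplay.
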NiagaraSystem /BRCosmetics/Animation/Game/MainPlayer/Emotes/Disconnect/FX/NS_Emote_Disconnect_Heart_Spawn
dependencies {
 usedDataInterfaces [&NiagaraDataInterfaceCurve, &NiagaraDataInterfaceColorCurve, &NiagaraDataInterfaceCamera]
 usedDynamicInputs []
 usedModules [&SystemState]
}
emitters {
 BG_Glow {
  emitterUpdateScript [
   &EmitterState(Life Cycle Mode=::NewEnumerator1, Loop Behavior=::NewEnumerator1, Loop Delay=0.05, Loop Duration=1, UseLoopDelay=true)
   &SpawnBurst_Instantaneous(Spawn Count=1, Use Spawn Probability=false)
  ]
  particleSpawnScript [
   &InitializeParticle(
    Color=<rgba 0.1769 0.0064 0.3073 1>
    Lifetime=0.25
    Sprite Size Mode=::NewEnumerator3
    Uniform Sprite Size=200
    UsePositionOffset=false
    Write Material Random=true
   )
   &CameraOffset(Camera Offset Amount=-20)
  ]
  particleUpdateScript [
   &ParticleState()
   &ScaleSpriteSize(Scale Factor=&Vector2DFromFloat(Value=&FloatFromCurve(
      FloatCurve=DataInterface() {propertyValues {Curve {keys [
          {arriveTangent=0.4346, interpMode=RCIM_Cubic, leaveTangent=0.4346, tangentMode=RCTM_User, time=0, value=0}
          {arriveTangent=0.0473, interpMode=RCIM_Cubic, leaveTangent=0.0473, tangentMode=RCTM_User, time=0.2, value=1}
         ]}}}
      Scale Curve=0.35
     )))
   &ScaleColor(
    Scale Alpha=&FloatFromCurve(
     FloatCurve=DataInterface() {propertyValues {Curve {keys [
         {arriveTangent=0.4005, interpMode=RCIM_Cubic, leaveTangent=0.4005, tangentMode=RCTM_User, time=0, value=0}
         {arriveTangent=0.0256, interpMode=RCIM_Cubic, leaveTangent=0.0256, tangentMode=RCTM_User, time=0.2, value=1}
         <at 1 0 RCIM_Cubic>
        ]}}}
     Scale Curve=&FNTimeOfDayScalar(Night Scale=0.8, Morning Scale=0.8, Evening Scale=0.8)
    )
    ScaleA=true
    ScaleRGB=true
   )
   /SetVariables(
    Particles.GlowColor2:LinearColor=&MakeLinearColorFromVectorAndFloat(
     "Vector (RGB)"=&MakeVectorFromLinearColorRGB(Color=<rgba 0.4427 0.6706 5 1>)
     "Float (Alpha)"=&Multiply_Float(
      A=0.2
      B=&FloatFromCurve(FloatCurve=DataInterface() {propertyValues {Curve {keys [
           {arriveTangent=0.4005, interpMode=RCIM_Cubic, leaveTangent=0.4005, tangentMode=RCTM_User, time=0, value=0}
           {arriveTangent=0.0256, interpMode=RCIM_Cubic, leaveTangent=0.0256, tangentMode=RCTM_User, time=0.2, value=1}
           <at 1 0 RCIM_Cubic>
          ]}}})
     )
    )
    Particles.GlowSize2:Vector2f=&Vector2DFromFloat(Value=&Multiply_Float(
      A=45
      B=&FloatFromCurve(
       FloatCurve=DataInterface() {propertyValues {Curve {keys [
           {arriveTangent=0.4346, interpMode=RCIM_Cubic, leaveTangent=0.4346, tangentMode=RCTM_User, time=0, value=0}
           {arriveTangent=0.0473, interpMode=RCIM_Cubic, leaveTangent=0.0473, tangentMode=RCTM_User, time=0.2, value=1}
          ]}}}
       Scale Curve=&FNTimeOfDayScalar(Night Scale=0.6, Morning Scale=0.4, Evening Scale=0.5)
      )
     ))
   ) {isSetParameters=true}
   /SetVariables(
    Particles.GlowColor03:LinearColor=&MultiplyLinearColor(
     Linear Color=&ColorFromCurve(ColorCurve=DataInterface() {propertyValues {
        AlphaCurve {keys [<at 0 1>,<at 0.1 0.7>,<at 1 0>]}
        BlueCurve {keys [<at 0 1.9193>,<at 0.1 1>,<at 1 1>]}
        GreenCurve {keys [<at 0 0.2604>,<at 0.1 0.0312>,<at 1 0.1719>]}
        RedCurve {keys [<at 0 5>,<at 0.1 0.1766>,<at 1 0.0625>]}
       }})
     Scale Factor=&MakeFloat4FromVectorAndFloat(
      "Vector (RGB)"=<1 1 1>
      "Float (Alpha)"=&FNTimeOfDayScalar(Night Scale=0.3, Morning Scale=0.3, Evening Scale=0.5)
     )
    )
    Particles.GlowSize03:Vector2f=&Vector2DFromFloat(Value=&FloatFromCurve(
      FloatCurve=DataInterface() {propertyValues {Curve {keys [{tangentMode=RCTM_User, time=0, value=0}, {tangentMode=RCTM_User, time=0.1, value=1}, <at 1 3>]}}}
      Scale Curve=75
     ))
   ) {isSetParameters=true}
  ]
  rendererClasses [&NiagaraSpriteRendererProperties]
  renderers [
   {rendererClass=&NiagaraSpriteRendererProperties}
   {rendererClass=&NiagaraSpriteRendererProperties, rendererIndex=1}
   {rendererClass=&NiagaraSpriteRendererProperties, rendererIndex=2}
  ]
 }
 Dust001 {
  emitterUpdateScript [
   &EmitterState(
    Enable Distance Culling=true
    MaxDistance=10000
    Module.bUseMaxDistance=true
    Module.bUseMinDistance=false
    Scalability Mode=::NewEnumerator1
   )
   &SpawnBurst_Instantaneous(Spawn Count=8, Use Spawn Probability=false, UseLoopCountLimit=false)
  ]
  particleSpawnScript [
   /SetVariables(Particles.SpriteRotationRate=&RandomRangeFloat(Minimum=-0.03, Maximum=0.03, Evaluation Type=::NewEnumerator1)) {isSetParameters=true}
   /Niagara/Modules/Spawn/Initialization/InitializeParticle(
    Lifetime=&RandomRangeFloat(Minimum=1, Maximum=1.5, Evaluation Type=::NewEnumerator1)
    Module.Write Color=true
    Module.Write Lifetime=true
    Module.Write Mass=true
    Module.Write Material Random=true
    Module.Write Position=false
    Module.Write Scale=false
    Module.Write SpriteRotation=true
    Module.Write SpriteSize=true
    Sprite Rotation=&RandomRangeFloat(Minimum=-360, Maximum=360, Evaluation Type=::NewEnumerator1)
    Sprite Size=&RandomRangeVector2D(Minimum=<80 80>, Maximum=<120 120>, Evaluation Type=::NewEnumerator1)
   ) {enabled=false}
   /Niagara/Modules/Spawn/Initialization/V2/InitializeParticle(
    Lifetime=&RandomRangeFloat(Minimum=0.25, Maximum=0.375)
    Mass Mode=::NewEnumerator1
    Sprite Rotation Mode=::NewEnumerator2
    Sprite Size=&RandomRangeVector2D(Minimum=<80 80>, Maximum=<120 120>)
    Sprite Size Mode=::NewEnumerator4
    UsePositionOffset=false
    Write Material Random=true
   ) {module=InitializeParticle001}
   &Color(Color=&MakeLinearColorFromVectorAndFloat(
     "Vector (RGB)"=&RandomRangeVector(Minimum=<0.3589 0.3009 0.4531>, Maximum=<0.605 0.6866 0.7396>, Evaluation Type=::NewEnumerator1)
     "Float (Alpha)"=&ClampFloat(Float=&RandomRangeFloat(Minimum=0.2, Maximum=0.3, Evaluation Type=::NewEnumerator1))
    )) {enabled=false}
   &Color(Color=&SelectLinearColorFromArray(Array Sampling Mode=::NewEnumerator0, Direct Array Index=?, Color Selection Array=$User.Color Array, Random Seed=0)) {module=Color001}
   &ShapeLocation(MaskSpawnGroup=false, Offset Mode=::NewEnumerator0, Override Seed=false, Sphere Radius=1, UseOwnerScale=true)
   &AddVelocity(
    Module.Use Falloff Distance=false
    Override Seed=false
    Velocity Mode=::NewEnumerator1
    Velocity Speed=&Multiply_Float(A=&RandomRangeFloat(Minimum=200, Maximum=500), B=0.5)
   )
   &AddVelocity(Velocity Speed Scale=&RandomRangeFloat(Minimum=0.5, Maximum=1)) {module=AddVelocity001}
   &CameraOffset(Camera Offset Amount=&RandomRangeFloat(Minimum=-10, Maximum=-50, Evaluation Type=::NewEnumerator1))
  ]
  particleUpdateScript [
   &ParticleState()
   &ScaleSpriteSize(
    Scale Sprite Size Mode=::NewEnumerator0
    Uniform Scale Factor=&FloatFromCurve(
     FloatCurve=DataInterface() {propertyValues {Curve {keys [
         {interpMode=RCIM_Cubic, tangentMode=RCTM_User, time=0, value=0.3}
         {arriveTangent=0.5158, interpMode=RCIM_Cubic, leaveTangent=0.5158, tangentMode=RCTM_User, time=0.4, value=0.9}
         {arriveTangentWeight=0.3824, interpMode=RCIM_Cubic, tangentMode=RCTM_User, time=1, value=1}
        ]}}}
     Scale Curve=0.75
    )
   )
   &Drag(Drag=10, Use Linear Drag=true, Use Rotational Drag=false)
   &PointAttractionForce(
    Attraction Radius=500
    Attraction Strength=&Lerp_Float(
     A=200
     B=1000
     Alpha=&FloatFromCurve(
      FloatCurve=DataInterface() {propertyValues {Curve {keys [
          {interpMode=RCIM_Cubic, tangentMode=RCTM_User, time=0, value=0}
          {arriveTangent=0.6382, interpMode=RCIM_Cubic, leaveTangent=0.6382, tangentMode=RCTM_User, time=0.65, value=0.2}
          {arriveTangentWeight=0.1753, interpMode=RCIM_Cubic, tangentMode=RCTM_User, time=1, value=1}
         ]}}}
      CurveIndex=&RandomRangeFloat(Minimum=0, Maximum=1)
     )
    )
    Kill Radius=5
    Kill Within Radius=true
    Module.Use Falloff=true
    UseAttractorOffset=false
   ) {enabled=false}
   &SolveForcesAndVelocity(Clamp Velocity=false, Limit Acceleration=false)
   &SpriteRotationRate(Rotation Rate=$Particles.SpriteRotationRate)
   &ScaleColor(
    Scale Alpha=&FloatFromCurve(
     FloatCurve=DataInterface() {propertyValues {Curve {keys [
         {arriveTangent=8.4225, interpMode=RCIM_Cubic, leaveTangent=8.4225, tangentMode=RCTM_User, time=0, value=0}
         {arriveTangentWeight=0.0721, interpMode=RCIM_Cubic, leaveTangentWeight=0.1081, tangentMode=RCTM_User, time=0.2, value=1}
         {arriveTangentWeight=0.1667, interpMode=RCIM_Cubic, leaveTangentWeight=0.1, tangentMode=RCTM_User, time=0.7, value=1}
         {arriveTangentWeight=0.7908, interpMode=RCIM_Cubic, tangentMode=RCTM_User, time=1, value=1}
        ]}}}
     Scale Curve=&Multiply_Float(A=0.2, B=&FNTimeOfDayScalar(Night Scale=0.2, Morning Scale=1, Evening Scale=1))
    )
    ScaleA=true
    ScaleRGB=true
   )
   &DynamicMaterialParameters(
    Index 0 Param 1=&Add_Float(
     A=&RandomRangeFloat(Minimum=0, Maximum=1)
     B=&FloatFromCurve(
      FloatCurve=DataInterface() {propertyValues {Curve {keys [
          {arriveTangent=2.7533, interpMode=RCIM_Cubic, leaveTangent=1.8958, tangentMode=RCTM_User, time=0, value=1}
          {arriveTangent=0.8292, interpMode=RCIM_Cubic, leaveTangent=0.8292, tangentMode=RCTM_User, time=0.321, value=1.3453}
          {arriveTangent=0.0529, interpMode=RCIM_Cubic, leaveTangent=0.2021, tangentMode=RCTM_User, time=1, value=1.5}
         ]}}}
      CurveIndex=&Add_Float(A=&Multiply_Float(A=$Particles.Initial.MaterialRandom, B=0.5), B=$Particles.NormalizedAge)
      Scale Curve=&Lerp_Float(A=1, B=2, Alpha=0.5)
     )
    )
    Index 0 Param 2=&FloatFromCurve(FloatCurve=DataInterface() {propertyValues {Curve {keys [<at 0 1>,<at 1 0>]}}})
    Index 0 Param 3=&FloatFromCurve(
     FloatCurve=DataInterface() {propertyValues {Curve {keys [
         {arriveTangent=-2.2052, interpMode=RCIM_Cubic, leaveTangent=-2.2052, tangentMode=RCTM_User, time=0.1, value=1}
         {interpMode=RCIM_Cubic, tangentMode=RCTM_User, time=1, value=0.4}
        ]}}}
     Scale Curve=0.5
    )
    Index 0 Param 4=&FloatFromCurve(FloatCurve=DataInterface() {propertyValues {Curve {keys [{interpMode=RCIM_Cubic, tangentMode=RCTM_Break, time=0.5, value=1}, {interpMode=RCIM_Cubic, tangentMode=RCTM_Break, time=1, value=0}]}}})
    Index 1 Param 1=&FloatFromCurve(
     FloatCurve=DataInterface() {propertyValues {Curve {keys [
         {interpMode=RCIM_Cubic, leaveTangentWeight=0.0833, tangentMode=RCTM_Break, time=0, value=0}
         {arriveTangentWeight=0.0873, interpMode=RCIM_Cubic, tangentMode=RCTM_Break, time=0.25, value=1}
        ]}}}
     Scale Curve=&Lerp_Float(A=0.35, B=0.6, Alpha=0.5)
    )
    Index 1 Param 2=&FloatFromCurve(
     FloatCurve=DataInterface() {propertyValues {Curve {keys [
         {interpMode=RCIM_Cubic, tangentMode=RCTM_Break, time=0, value=0.2}
         {interpMode=RCIM_Cubic, leaveTangent=0.9066, tangentMode=RCTM_Break, time=1, value=2}
        ]}}}
     Scale Curve=&RandomRangeFloat(Minimum=0.1, Maximum=0.2)
    )
    Index 1 Param 3=&FloatFromCurve(FloatCurve=DataInterface() {propertyValues {Curve {keys [
         {interpMode=RCIM_Cubic, leaveTangent=6.3917, tangentMode=RCTM_Break, time=0, value=0}
         {arriveTangent=1.7549, interpMode=RCIM_Cubic, leaveTangent=10.6767, tangentMode=RCTM_User, time=0.25, value=1}
        ]}}})
    Index 1 Param 4=&FloatFromCurve(
     FloatCurve=DataInterface() {propertyValues {Curve {keys [{arriveTangent=-0.136, interpMode=RCIM_Cubic, leaveTangent=-0.136, tangentMode=RCTM_User, time=0, value=1}, <at 1 0.95>]}}}
     Scale Curve=&Lerp_Float(A=2, B=5, Alpha=0.5)
    )
    Index 2 Param 1=&FloatFromCurve(FloatCurve=DataInterface() {propertyValues {Curve {keys [
         {interpMode=RCIM_Cubic, leaveTangent=4.2485, tangentMode=RCTM_Break, time=0, value=0}
         {arriveTangent=1.0548, interpMode=RCIM_Cubic, leaveTangent=1.0548, tangentMode=RCTM_User, time=0.17, value=0.5715}
         <at 1 1>
        ]}}})
    Index 2 Param 2=&FloatFromCurve(
     FloatCurve=DataInterface() {propertyValues {Curve {keys [
         {arriveTangent=1.9056, interpMode=RCIM_Cubic, leaveTangent=1.9056, tangentMode=RCTM_User, time=0, value=0.2}
         {arriveTangent=0.0259, interpMode=RCIM_Cubic, leaveTangent=0.0259, tangentMode=RCTM_User, time=1, value=1.4}
        ]}}}
     Scale Curve=1.5
    )
    Index 2 Param 3=&FloatFromCurve(
     FloatCurve=DataInterface() {propertyValues {Curve {keys [
         {interpMode=RCIM_Cubic, leaveTangentWeight=0.1411, tangentMode=RCTM_Break, time=0, value=2}
         {interpMode=RCIM_Cubic, tangentMode=RCTM_Break, time=0.3, value=1}
         {interpMode=RCIM_Cubic, leaveTangent=11.1233, tangentMode=RCTM_Break, time=1, value=2}
        ]}}}
     Scale Curve=10
    )
    Index 2 Param 4=&FloatFromCurve(
     FloatCurve=DataInterface() {propertyValues {Curve {keys [
         {arriveTangent=-0.6867, interpMode=RCIM_Cubic, leaveTangent=-0.6867, tangentMode=RCTM_User, time=0, value=1}
         {interpMode=RCIM_Cubic, leaveTangent=-1.2956, tangentMode=RCTM_Break, time=1, value=0.75}
        ]}}}
     Scale Curve=&RandomRangeFloat(Minimum=0.5, Maximum=2)
    )
    Module.Param0WriteEnabled=true
    Module.Param0WriteEnabled1=true
    Module.Param0WriteEnabled2=true
    Module.Param1WriteEnabled=false
    Module.Param1WriteEnabled1=true
    Module.Param1WriteEnabled2=true
    Module.Param2WriteEnabled=true
    Module.Param2WriteEnabled1=true
    Module.Param2WriteEnabled2=true
    Module.Param3WriteEnabled=true
    Module.Param3WriteEnabled1=true
    Module.Param3WriteEnabled2=true
    Write Parameter Index 1=true
    Write Parameter Index 2=true
   )
  ]
  rendererClasses [&NiagaraSpriteRendererProperties]
  renderers [
   {rendererClass=&NiagaraSpriteRendererProperties}
   {rendererClass=&NiagaraSpriteRendererProperties, rendererIndex=1}
  ]
 }
 Empty {
  emitterUpdateScript [
   &EmitterState()
   &SpawnBurst_Instantaneous(Use Spawn Probability=false, UseLoopCountLimit=false)
  ]
  particleSpawnScript [
   &InitializeParticle(
    Color=<rgba 1 0 0 0.75>
    Lifetime=0.667
    Mesh Scale Mode=::NewEnumerator3
    Mesh Uniform Scale=1.1
    Position Offset=<0 0 -24.6>
    Sprite Size Mode=::NewEnumerator3
    Uniform Sprite Size=150
    UsePositionOffset=true
    Write Material Random=true
   )
   &InitializeParticle(
    Color=<rgba 1 0 0 0.75>
    Lifetime=0.667
    Mesh Scale Mode=::NewEnumerator3
    Mesh Uniform Scale=1.1
    Position Offset=<5 0 -24>
    Sprite Size Mode=::NewEnumerator3
    Uniform Sprite Size=150
    UsePositionOffset=true
    Write Material Random=true
   ) {enabled=false, module=InitializeParticle001}
   &InitialMeshOrientation(
    Mesh Orientation Mode=::NewEnumerator7
    Rotation=&MakeVector(
     X=&AngleConversion(Angle:NiagaraFloat=0, Angle Input:ENiagara_AngleInput=::NewEnumerator0, Angle Output:ENiagara_AngleInput=::NewEnumerator1)
     Y=&AngleConversion(Angle:NiagaraFloat=0, Angle Input:ENiagara_AngleInput=::NewEnumerator0, Angle Output:ENiagara_AngleInput=::NewEnumerator1)
     Z=&AngleConversion(Angle:NiagaraFloat=0, Angle Input:ENiagara_AngleInput=::NewEnumerator0, Angle Output:ENiagara_AngleInput=::NewEnumerator1)
    )
    Use Rotation Vector=true
   )
   &InitialMeshOrientation(
    Mesh Orientation Mode=::NewEnumerator7
    Rotation=&MakeVector(
     X=&AngleConversion(Angle:NiagaraFloat=0, Angle Input:ENiagara_AngleInput=::NewEnumerator0, Angle Output:ENiagara_AngleInput=::NewEnumerator1)
     Y=&AngleConversion(Angle:NiagaraFloat=-14.088, Angle Input:ENiagara_AngleInput=::NewEnumerator0, Angle Output:ENiagara_AngleInput=::NewEnumerator1)
     Z=&AngleConversion(Angle:NiagaraFloat=0, Angle Input:ENiagara_AngleInput=::NewEnumerator0, Angle Output:ENiagara_AngleInput=::NewEnumerator1)
    )
    Use Rotation Vector=true
   ) {enabled=false, module=InitialMeshOrientation001}
   &CameraOffset(Camera Offset Amount=-25)
   /SetVariables() {isSetParameters=true}
  ]
  particleUpdateScript [
   &ParticleState()
   &ScaleSpriteSize(Uniform Curve Sprite Scale=DataInterface() {propertyValues {Curve {keys [<at 0 0 RCIM_Cubic>, {arriveTangentWeight=0.05, interpMode=RCIM_Cubic, tangentMode=RCTM_User, time=0.15, value=1}]}}})
   &ScaleMeshSize(Scale Factor=&VectorFromFloat(Value=&FloatFromCurve(FloatCurve=DataInterface() {propertyValues {Curve {keys [{tangentMode=RCTM_User, time=0, value=0}, {tangentMode=RCTM_User, time=0.05, value=1}]}}})))
   /SetVariables(Particles.ComponentPosition:NiagaraPosition=&AddVectorToPosition(
     Position=$Particles.Position
     Vector=&Multiply_VectorByFloat(
      Vector=&ReturnCameraVector(Camera Vectors=::NewEnumerator4, Coordinate Space=::Simulation, Origin Position=$Particles.Position)
      Float=2
     )
    )) {enabled=false, isSetParameters=true}
   /SetVariables(Particles.DisplayPosition:NiagaraPosition=<-2 0 2>) {isSetParameters=true}
  ]
  rendererClasses [&NiagaraSpriteRendererProperties, &NiagaraMeshRendererProperties]
  renderers [
   {rendererClass=&NiagaraSpriteRendererProperties}
   {rendererClass=&NiagaraMeshRendererProperties, rendererIndex=1}
  ]
 }
 Flare_With_Streak {
  emitterUpdateScript [
   &EmitterState(Life Cycle Mode=::NewEnumerator1, Loop Behavior=::NewEnumerator1, Loop Duration=1, UseLoopDelay=false)
   &SpawnBurst_Instantaneous(Spawn Count=1, Use Spawn Probability=false)
  ]
  particleSpawnScript [
   &InitializeParticle(
    Color=<rgba 2 1.0873 0 3>
    Lifetime=0.2
    Sprite Rotation Mode=::NewEnumerator2
    Sprite Size Mode=::NewEnumerator3
    Uniform Sprite Size=200
    UsePositionOffset=false
    Write Material Random=true
   )
   &DynamicMaterialParameters(
    Index 0 Param 1=&RandomRangeFloat(Minimum=0.4, Maximum=0.6)
    Index 0 Param 2=1
    Index 0 Param 3=1
    Index 0 Param 4=1
    Module.Param0WriteEnabled=true
    Module.Param1WriteEnabled=true
    Module.Param2WriteEnabled=true
    Module.Param3WriteEnabled=true
   )
   &SystemLocation(MaskSpawnGroup=false)
   /SetVariables() {isSetParameters=true}
   &CameraOffset(Camera Offset Amount=25)
  ]
  particleUpdateScript [
   &ParticleState()
   &SolveForcesAndVelocity(Clamp Velocity=false, Limit Acceleration=false)
   &ScaleColor(
    Scale Alpha=&Multiply_Float(
     A=1
     B=&FloatFromCurve(FloatCurve=DataInterface() {propertyValues {Curve {keys [<at 0 0>,<at 0.2 1>,<at 0.8 1>,<at 1 0>]}}})
    )
    ScaleA=true
    ScaleRGB=true
   )
   &ScaleSpriteSize(Scale Factor=&Vector2DFromFloat(Value=&FloatFromCurve(
      FloatCurve=DataInterface() {propertyValues {Curve {keys [<at 0 1.5 RCIM_Cubic>,<at 1 0>]}}}
      Scale Curve=&FN_FrontEndFloatModifier(Gameplay=1, Front End=0.7)
     )))
   /SetVariables(Particles.Flash size:Vector2f=&Multiply_Vector2D_ByFloat(
     Float=&Multiply_Float(
      A=400
      B=&FloatFromCurve(
       FloatCurve=DataInterface() {propertyValues {Curve {keys [
           {arriveTangent=4.1785, interpMode=RCIM_Cubic, leaveTangent=4.1785, tangentMode=RCTM_User, time=0, value=0}
           {arriveTangent=0.9995, interpMode=RCIM_Cubic, leaveTangent=0.9995, tangentMode=RCTM_User, time=0.1378, value=0.6825}
           {arriveTangent=-0.0405, interpMode=RCIM_Cubic, leaveTangent=-0.0405, tangentMode=RCTM_User, time=0.62, value=1}
          ]}}}
       Scale Curve=&FN_FrontEndFloatModifier(Gameplay=0.9, Front End=0.7)
      )
     )
     Vector2D=&Multiply_Vector2D_ByFloat(Float=0.35, Vector2D=<1.5 1>)
    )) {isSetParameters=true}
   /SetVariables(Particles.Flash color:LinearColor=&MakeLinearColorFromVectorAndFloat(
     "Vector (RGB)"=&MakeVectorFromLinearColorRGB(Color=<rgba 1 0.5585 0 1>)
     "Float (Alpha)"=&Multiply_Float(
      A=&Multiply_Float(
       A=1
       B=&FloatFromCurve(FloatCurve=DataInterface() {propertyValues {Curve {keys [<at 0 0>,<at 0.124 1>,<at 0.496 1>,<at 0.62 0>]}}})
      )
      B=1
     )
    )) {isSetParameters=true}
  ]
  rendererClasses [&NiagaraSpriteRendererProperties]
  renderers [
   {rendererClass=&NiagaraSpriteRendererProperties}
   {rendererClass=&NiagaraSpriteRendererProperties, rendererIndex=1}
  ]
 }
 Light {
  emitterUpdateScript [
   &EmitterState()
   &SpawnBurst_Instantaneous(Use Spawn Probability=false, UseLoopCountLimit=false)
  ]
  particleSpawnScript [&InitializeParticle(Color=<rgba 1.3809 0.2419 3 1>, Lifetime=0.25, UsePositionOffset=false, Write Material Random=true)]
  particleUpdateScript [
   &ParticleState()
   &ScaleColor(
    Scale Alpha=&FloatFromCurve(FloatCurve=DataInterface() {propertyValues {Curve {keys [<at 0 1>,<at 1 0>]}}})
    ScaleA=true
    ScaleRGB=false
   )
   /SetVariables(Particles.LightColor=&MultiplyLinearColor(
     Linear Color=$Particles.Color
     Scale Factor=&MakeFloat4FromVectorAndFloat("Vector (RGB)"=<0.333 1 1>, "Float (Alpha)"=80)
    )) {isSetParameters=true}
  ]
  rendererClasses [&NiagaraLightRendererProperties]
  renderers [{rendererClass=&NiagaraLightRendererProperties, rendererIndex=0}]
 }
 SoftShockwave {
  emitterUpdateScript [
   &EmitterState()
   &SpawnBurst_Instantaneous(Use Spawn Probability=false, UseLoopCountLimit=false)
  ]
  particleSpawnScript [&InitializeParticle(
    Color=<rgba 0.5319 0.9365 4 3>
    Lifetime=0.2
    Mesh Scale Mode=::NewEnumerator3
    Mesh Uniform Scale=0.8732
    UsePositionOffset=false
    Write Material Random=true
   )]
  particleUpdateScript [
   &ParticleState()
   &ScaleColor(
    Scale Alpha=&FloatFromCurve(FloatCurve=DataInterface() {propertyValues {Curve {keys [<at 0 1>, {interpMode=RCIM_Cubic, tangentMode=RCTM_User, time=0.1, value=1}, {interpMode=RCIM_Cubic, tangentMode=RCTM_User, time=1, value=0}]}}})
    ScaleA=true
    ScaleRGB=true
   )
   &DynamicMaterialParameters(
    Index 0 Param 1=1
    Index 0 Param 2=-0.5
    Index 0 Param 3=1
    Index 0 Param 4=1
    Module.Param0WriteEnabled=true
    Module.Param1WriteEnabled=true
    Module.Param2WriteEnabled=true
    Module.Param3WriteEnabled=true
   )
   &ScaleMeshSize(Scale Factor=&VectorFromFloat(Value=&FloatFromCurve(FloatCurve=DataInterface() {propertyValues {Curve {keys [
          {arriveTangent=4.1785, interpMode=RCIM_Cubic, leaveTangent=4.1785, tangentMode=RCTM_User, time=0, value=0.05}
          {arriveTangent=0.9995, interpMode=RCIM_Cubic, leaveTangent=0.9995, tangentMode=RCTM_User, time=0.3, value=0.6}
          {arriveTangent=-0.0405, interpMode=RCIM_Cubic, leaveTangent=-0.0405, tangentMode=RCTM_User, time=1, value=1}
         ]}}})))
  ]
  rendererClasses [&NiagaraMeshRendererProperties]
  renderers [{rendererClass=&NiagaraMeshRendererProperties, rendererIndex=0}]
 }
 shine {
  emitterUpdateScript [
   &EmitterState(Life Cycle Mode=::NewEnumerator1, Loop Behavior=::NewEnumerator1, Loop Delay=0.08, Loop Duration=1, UseLoopDelay=true)
   &SpawnBurst_Instantaneous(Spawn Count=12, Use Spawn Probability=false)
  ]
  particleSpawnScript [
   &InitializeParticle(
    Color=&RandomRangeLinearColor(Minimum=<rgba 1 1 1 0.5>, Maximum=<rgba 2 2 2 1.2>)
    Lifetime Max=0.15
    Lifetime Min=0.1
    Lifetime Mode=::NewEnumerator1
    Sprite Rotation Mode=::NewEnumerator2
    Sprite Size Max=<100 200>
    Sprite Size Min=<60 100>
    Sprite Size Mode=::NewEnumerator2
    UsePositionOffset=false
    Write Material Random=true
   )
   &SubUVAnimation(End Frame=3, SubUV Animation Mode:ENiagara_SubUVLookupMode=::NewEnumerator2)
  ]
  particleUpdateScript [
   &ParticleState()
   &ScaleSpriteSize(Scale Factor=<0.5 0.3>, Scale Sprite Size Mode=::NewEnumerator2) {module=ScaleSpriteSize001}
   &ScaleSpriteSize(Scale Factor=&MakeVector2D(
     X=&FloatFromCurve(FloatCurve=DataInterface() {propertyValues {Curve {keys [
          {interpMode=RCIM_Cubic, tangentMode=RCTM_User, time=0, value=0}
          {interpMode=RCIM_Cubic, tangentMode=RCTM_User, time=0.2, value=1}
          {tangentMode=RCTM_User, time=1, value=0.5}
         ]}}})
     Y=&FloatFromCurve(FloatCurve=DataInterface() {propertyValues {Curve {keys [
          <at 0 0 RCIM_Cubic>
          {arriveTangent=1.5, interpMode=RCIM_Cubic, leaveTangent=1.5, time=0.2, value=0.9}
          {tangentMode=RCTM_User, time=1, value=1.5}
         ]}}})
    ))
   &ScaleColor(
    Linear Color Curve=DataInterface() {propertyValues {
      AlphaCurve {keys [<at 0.1263 1>,<at 1 0>]}
      BlueCurve {keys [<at 0 20>,<at 0.1162 2>,<at 1 0.3635>]}
      GreenCurve {keys [<at 0 3.75>,<at 0.1162 0.1042>,<at 1 0.0208>]}
      RedCurve {keys [<at 0 1.25>,<at 0.1162 0.1674>,<at 1 1>]}
     }}
    Scale Mode=::NewEnumerator2
   ) {enabled=false}
   &ScaleColor(
    Linear Color Curve=DataInterface() {propertyValues {
      AlphaCurve {keys [<at 0.333 1>,<at 1 0>]}
      BlueCurve {keys [<at 0 20>,<at 0.1162 1>,<at 1 1>]}
      GreenCurve {keys [<at 0 3.75>,<at 0.1162 1>,<at 1 1>]}
      RedCurve {keys [<at 0 1.25>,<at 0.1162 1>,<at 1 1>]}
     }}
    Scale Mode=::NewEnumerator2
   ) {module=ScaleColor002}
   &ScaleColor(
    Color Value To Scale=$StackContext.Color
    Scale RGB=&MakeVectorFromLinearColorRGB(Color=&SelectLinearColorFromArray(Array Sampling Mode=::NewEnumerator0, Direct Array Index=?, Color Selection Array=$User.Color Array, Random Seed=0))
    ScaleA=false
    ScaleRGB=true
   ) {module=ScaleColor003}
   &ScaleColor(
    Scale Alpha=&FNTimeOfDayScalar(Night Scale=0.5, Morning Scale=0.3, Evening Scale=0.5)
    Scale RGB=&VectorFromFloat(Value=&FloatFromCurve(FloatCurve=DataInterface() {propertyValues {Curve {keys [{arriveTangent=-53.2404, interpMode=RCIM_Cubic, leaveTangent=-53.2404, tangentMode=RCTM_User, time=0, value=20}, <at 0.3 0 RCIM_Cubic>]}}}))
    ScaleA=true
    ScaleRGB=false
   ) {module=ScaleColor001}
  ]
  rendererClasses [&NiagaraSpriteRendererProperties]
  renderers [{rendererClass=&NiagaraSpriteRendererProperties}]
 }
}
summary {
 systemName=NS_Emote_Disconnect_Heart_Spawn
 emitters [
  {emitterName=Light, rendererClasses [&NiagaraLightRendererProperties]}
  {emitterName=SoftShockwave, rendererClasses [&NiagaraMeshRendererProperties]}
  {emitterName=Flare_With_Streak, rendererClasses [&NiagaraSpriteRendererProperties]}
  {emitterName=BG_Glow, rendererClasses [&NiagaraSpriteRendererProperties]}
  {emitterName=shine, rendererClasses [&NiagaraSpriteRendererProperties]}
  {emitterName=Dust001, rendererClasses [&NiagaraSpriteRendererProperties]}
  {emitterName=Empty, rendererClasses [&NiagaraSpriteRendererProperties, &NiagaraMeshRendererProperties]}
 ]
 userVariables [
  User.bIsFrontEnd {type=NiagaraBool}
  User.bIsFrontEndPreview {type=NiagaraBool}
  User.Color Array {
   type=NiagaraDataInterfaceArrayColor
   description=""
   defaultValue {
    struct=NiagaraExt_VariableValue_DataInterface
    value {
     dataInterface=&self:NiagaraDataInterfaceArrayColor_0
     dataInterfaceClass=&NiagaraDataInterfaceArrayColor
    }
   }
  }
  User.DynamicFloat {type=NiagaraFloat, defaultValue=0, description=""}
  User.Offset {type=Vector3f, description="", defaultValue {struct=Vector3f, value=<0 0 0>}}
  User.SkeletalMesh {
   type=NiagaraDataInterfaceSkeletalMesh
   defaultValue {
    struct=NiagaraExt_VariableValue_DataInterface
    value {dataInterface=&self:NiagaraDataInterfaceSkeletalMesh_0}
   }
  }
 ]
}
